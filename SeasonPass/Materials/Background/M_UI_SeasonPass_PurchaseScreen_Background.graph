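# Translucent UI material for the Season Pass purchase screen background, built around an Image texture parameter that defaults to a transparent blank.
# MF_UI_ScreenSpaceUVScaling uses the texture's size and a centered pivot to produce scale-to-fit screen-space UVs.
# The image is sampled with clamped addressing at those UVs, and its alpha and the UVs feed MF_UI_ClipEdgePixel to clip edge pixels.
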Material /SeasonPass/Materials/Background/M_UI_SeasonPass_PurchaseScreen_Background
graph {
 properties {
  bAutomaticallySetUsageInEditor=false
  bUsedWithStaticMesh=false
  blendMode=BLEND_Translucent
  materialDomain=MD_UI
  usedShadingModels=Default Lit
 }
 nodes [
  n1 = TextureObjectParameter() {parameterName=Image, texture=/Game/Global/Textures/Default/Blanks/T_Blank_White_ZeroAlpha_sRGBColor}
  n2 = MaterialFunctionCall!(ScreenUV, TextureSize=TextureProperty(Tex=@n1), ScalePivotPoint=Constant2Vector() {g=0.5, r=0.5}) {materialFunction=/FortUILibrary/MaterialFunctions/Basics/MF_UI_ScreenSpaceUVScaling}
  MaterialFunctionCall!(UVs=@n2.Scaled to Fit, Alpha=TextureSample(UVs=@n2.Scaled to Fit, Tex=@n1) {samplerSource=SSM_Clamp_WorldGroupSettings}.A) {materialFunction=/FortUILibrary/MaterialFunctions/Basics/MF_UI_ClipEdgePixel}
 ]
}
